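# Idle Niagara effect for the NoiseClueGust backpack cosmetic: audio-reactive fire with smoke, glow, sparks and lights.
# Socket-placed emitters build mesh flames, edge and wheel flames, and rack decals from persistent burst particles, plus continuously spawned sparks.
# System.IncomingAudio drives flame size, brightness, light falloff and spark rate, with a fake audio loop module in use.
# User.HideFX suppresses spawning in every emitter.
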NiagaraSystem /BRCosmetics/Accessories/FORT_Backpacks/Backpack_M_MED_NoiseClueGust/FX/NS_Backpack_NoiseClueGust_Idle
dependencies {
 usedDataInterfaces [
  &NiagaraDataInterfaceArrayInt32
  &NiagaraDataInterfaceColorCurve
  &NiagaraDataInterfaceCurve
  &NiagaraDataInterfaceArrayFloat
  /Script/Niagara.NiagaraDataInterfaceArrayQuat
  &NiagaraDataInterfaceArrayFloat3
  &NiagaraDataInterfaceArrayFloat2
  &NiagaraDataInterfaceSocketReader
 ]
 usedDynamicInputs [
  &MakeCustomFloatFromBool
  &SetBoolByFloatComparison
  &RemapRange
  &Waveform
  &MakeFloatFromLinearColor
 ]
 usedModules [
  &SystemState
  &self:SystemSpawnScript.SystemScriptSource.SystemScriptGraph.NiagaraNodeAssignment_1.SetVariables_#1
  &BlendSpaceHelper
  /Game/Effects/Niagara/Modules/Audio/FN_FakeAudioReactivityLoop
  &self:SystemSpawnScript.SystemScriptSource.SystemScriptGraph.NiagaraNodeAssignment_0.SetVariables_#2
  &self:SystemSpawnScript.SystemScriptSource.SystemScriptGraph.NiagaraNodeAssignment_2.SetVariables_#3
  &self:SystemSpawnScript.SystemScriptSource.SystemScriptGraph.NiagaraNodeAssignment_3.SetVariables_#4
  &Timeline
 ]
}
emitters {
 FlameEdge {
  emitterUpdateScript [
   &EmitterState(Life Cycle Mode=::NewEnumerator1, Loop Behavior=::NewEnumerator1, Loop Duration Mode=::NewEnumerator1, UseLoopDelay=false)
   &SpawnBurst_Instantaneous(
    Spawn Count=&Multiply_Int(
     A=1
     B=&MakeIntFromFloat(Float=&ClampFloat(Float=&OneMinusFloat(Float=$User.HideFX)))
    )
    Use Spawn Probability=false
    UseLoopCountLimit=false
   )
  ]
  particleSpawnScript [&InitializeParticle(
    Color=<rgba 2 0.3688 0.0744 0.4>
    Mesh Scale Mode=::NewEnumerator3
    Sprite Rotation Angle=-73.0845
    Sprite Rotation Mode=::NewEnumerator0
    Sprite Size=<20 15>
    Sprite Size Mode=::NewEnumerator4
    UsePositionOffset=false
    Write Material Random=true
   )]
  particleUpdateScript [
   &ParticleState(Kill Particles When Lifetime Has Elapsed=false)
   &ScaleSpriteSize(
    Uniform Curve Index=$System.NormalizedLoopAge
    Uniform Curve Scale=&Lerp_Float(A=0.8, B=1.2, Alpha=$System.IncomingAudio)
    Uniform Curve Sprite Scale=DataInterface() {propertyValues {Curve {keys [
        <at 0 1 RCIM_Cubic>
        {arriveTangent=-0.0833, interpMode=RCIM_Cubic, leaveTangent=-0.0833, time=0.2, value=1.1}
        {arriveTangent=-0.125, interpMode=RCIM_Cubic, leaveTangent=-0.125, time=0.6, value=0.95}
        <at 1 1 RCIM_Cubic>
       ]}}}
   )
   &SocketLocation(
    Orientation Sampling=::NewEnumerator1
    SamplePosOffset=true
    Sampled Position Offset=<-0.1915 -0.35 0>
    Socket Index=8
    Socket Reader=$System.Socket Reader
    Socket Sampling Mode=::NewEnumerator3
    VelocitySpeedLimit=false
   )
   &DynamicMaterialParameters(
    Index 0 Param 2=0.8
    Module.Index0Float0Write=true
    Module.Index0Float1Write=true
    Module.Index0Float2Write=true
    Module.Index0Float3Write=true
   )
   &ScaleColor(
    Scale Alpha=0.2
    Scale RGB=&Multiply_VectorByFloat(Vector=<1 1 1>, Float=&Lerp_Float(A=1, B=2, Alpha=$System.IncomingAudio))
    ScaleA=true
    ScaleRGB=true
   )
   &SpriteFacingAndAlignment(Sprite Alignment=<0 0 1>, Sprite Facing=<0 -1 0>, WriteAlignment=true, WriteFacing=true)
   /SetVariables(Particles.DMPFlame:Vector4f=<0.8 0.25 0 0>) {isSetParameters=true}
   /SetVariables(
    Particles.CameraOffset2=1
    Particles.FlameAlignment2:Vector3f=<0 0 1>
    Particles.FlamePosition:NiagaraPosition=&AddVectorToPosition(Position=$Output.SocketLocation.SampledPosition, Vector=<1.5213 -0.752 1.0334>)
    Particles.FlamePosition2:NiagaraPosition=&AddVectorToPosition(Position=$Output.SocketLocation.SampledPosition, Vector=<0 -1.4843 4>)
    Particles.SpriteRotation2=-90
    Particles.SpriteSize2=<15 4>
   ) {isSetParameters=true}
   /SetVariables(
    Particles.DMPSmoke:Vector4f=<0.2 0.2 0 0>
    Particles.SmokeSize:Vector2f=&Multiply_Vector2D_ByFloat(Float=1.2, Vector2D=$Particles.Initial.SpriteSize)
   ) {isSetParameters=true}
  ]
  rendererClasses [&NiagaraMeshRendererProperties, &NiagaraSpriteRendererProperties]
  renderers [
   {rendererClass=&NiagaraMeshRendererProperties, rendererIndex=0}
   {rendererClass=&NiagaraSpriteRendererProperties, rendererIndex=1}
   {rendererClass=&NiagaraSpriteRendererProperties, rendererIndex=2}
   {rendererClass=&NiagaraSpriteRendererProperties, rendererIndex=3}
   {rendererClass=&NiagaraSpriteRendererProperties, rendererIndex=4}
  ]
 }
 FlameMeshFront {
  emitterUpdateScript [
   &EmitterState(Life Cycle Mode=::NewEnumerator1, Loop Behavior=::NewEnumerator1, Loop Duration Mode=::NewEnumerator1, UseLoopDelay=false)
   &SpawnBurst_Instantaneous(
    Spawn Count=&Multiply_Int(
     A=2
     B=&MakeIntFromFloat(Float=&ClampFloat(Float=&OneMinusFloat(Float=$User.HideFX)))
    )
    Use Spawn Probability=false
    UseLoopCountLimit=false
   )
  ]
  particleSpawnScript [
   &InitializeParticle(
    Lifetime=5
    Mesh Scale Mode=::NewEnumerator4
    Sprite Size=<11 22>
    Sprite Size Mode=::NewEnumerator4
    UsePositionOffset=false
    Write Material Random=true
   )
   &PartitionParticles()
   &InitialMeshOrientation(Mesh Axis To Orient=<0 1 0>, Mesh Orientation Mode=::NewEnumerator2, Orientation Vector=<0 1 0>, Use Rotation Vector=false)
  ]
  particleUpdateScript [
   &ParticleState(Kill Particles When Lifetime Has Elapsed=false)
   &SocketLocation(
    Orientation Sampling=::NewEnumerator1
    SamplePosOffset=false
    Socket Index=&SelectIntFromArray(Direct Array Index=$Particles.PartitionParticles.Partition, Int Selection Array=DataInterface() {propertyValues {IntData [0,1]}})
    Socket Reader=$System.Socket Reader
    Socket Sampling Mode=::NewEnumerator3
    VelocitySpeedLimit=false
   )
   &DynamicMaterialParameters(
    Index 0 Param 1=0.8
    Index 0 Param 2=0.25
    Module.Index0Float0Write=true
    Module.Index0Float1Write=true
    Module.Index0Float2Write=true
    Module.Index0Float3Write=true
   )
   /SetVariables(
    Particles.MeshSizeFire:Vector3f=&MakeVector(
     X=&Multiply_Float(
      A=1
      B=&SelectFloatFromArray(Direct Array Index=$Particles.PartitionParticles.Partition, Float Selection Array=DataInterface() {propertyValues {FloatData [1,0.8]}})
     )
     Y=1
     Z=&Multiply_Float(
      A=&Multiply_Float(
       A=1
       B=&SelectFloatFromArray(Direct Array Index=$Particles.PartitionParticles.Partition, Float Selection Array=DataInterface() {propertyValues {FloatData [1.2,0.8]}})
      )
      B=&Lerp_Float(A=0.8, B=1.45, Alpha=$System.IncomingAudio)
     )
    )
    Particles.MeshSizeSmoke:Vector3f=&MakeVector(
     X=&Multiply_Float(
      A=1
      B=&SelectFloatFromArray(Direct Array Index=$Particles.PartitionParticles.Partition, Float Selection Array=DataInterface() {propertyValues {FloatData [1,0.8]}})
     )
     Y=1
     Z=&Multiply_Float(
      A=&Multiply_Float(
       A=1.4
       B=&SelectFloatFromArray(Direct Array Index=$Particles.PartitionParticles.Partition, Float Selection Array=DataInterface() {propertyValues {FloatData [1.2,1]}})
      )
      B=&Lerp_Float(A=0.8, B=1.4, Alpha=$System.IncomingAudio)
     )
    )
   ) {isSetParameters=true}
   /SetVariables(Particles.DMPDecal:Vector4f=<0 0.8 0 0>, Particles.DMPSmoke:Vector4f=<0.2 0.2 0 0>) {isSetParameters=true}
   &ScaleColor(
    Scale RGB=&Multiply_VectorByFloat(Vector=<1 1 1>, Float=&Lerp_Float(A=1, B=1.3, Alpha=$System.IncomingAudio))
    ScaleA=true
    ScaleRGB=true
   )
   /SetVariables(
    Particles.DecalColor:LinearColor=&MakeLinearColorFromVectorAndFloat(
     "Vector (RGB)"=&MakeVectorFromLinearColorRGB(Color=&Multiply_LinearColorByFloat(Linear Color=<rgba 2 0.2265 0.0045 0.4>, Float=&Lerp_Float(A=1, B=2, Alpha=$System.IncomingAudio)))
     "Float (Alpha)"=0.3
    )
    Particles.DecalPosition:NiagaraPosition=&AddVectorToPosition(Position=$Output.SocketLocation.SampledPosition, Vector=<0 -0.15 0>)
    Particles.DecalScale:Vector3f=<1.01 1.01 1.01>
    Particles.MeshOrientationDecal:Quat4f=/Niagara/DynamicInputs/Arrays/SelectQuaternionFromArray(
     Array Sampling Mode:ENiagaraArraySamplingMode=::NewEnumerator1
     Direct Array Index:NiagaraInt32=$Particles.PartitionParticles.Partition
     Interpolation:NiagaraFloat=?
     Direct Array Mode:ENiagara_WrapClamp=::NewEnumerator0
     Evaluation Type:ENiagaraRandomnessEvaluation=::NewEnumerator0
     Normalized Interpolation:NiagaraBool=true
     Recalculate Random Each Loop:NiagaraBool=false
     Randomness Mode:ENiagaraRandomnessMode=::NewEnumerator0
     Random Seed:NiagaraInt32=?
     Fixed Random Seed:NiagaraBool=false
     Quaternion Selection Array:NiagaraDataInterfaceArrayQuat=DataInterface() {propertyValues {GpuSyncMode=SyncCpuToGpu, MaxElements=0, QuatData [<0 0 0 1>, <0 -1 0 0>]}}
     Enable Spherical Linear Interpolation:NiagaraBool=?
    )
   ) {isSetParameters=true}
   /SetVariables(
    Particles.GlowAlignment:Vector3f=<0 0 1>
    Particles.GlowColor=&MakeLinearColorFromVectorAndFloat(
     "Vector (RGB)"=&MakeVectorFromLinearColorRGB(Color=<rgba 3 0.3313 0 1>)
     "Float (Alpha)"=&FloatFromCurve(
      FloatCurve=DataInterface() {propertyValues {Curve {keys [
          <at -0.0265 1.047 RCIM_Cubic>
          {arriveTangent=0.8137, interpMode=RCIM_Cubic, leaveTangent=0.8137, time=0.1748, value=0.6398}
          {arriveTangent=0.8749, interpMode=RCIM_Cubic, leaveTangent=0.8749, time=0.3787, value=1.3767}
          {arriveTangent=-0.6399, interpMode=RCIM_Cubic, leaveTangent=-0.6399, time=0.5551, value=0.9725}
          {arriveTangent=0.1779, interpMode=RCIM_Cubic, leaveTangent=0.1779, time=0.753, value=1.1372}
          <at 0.9735 1.047 RCIM_Cubic>
         ]}}}
      CurveIndex=$System.NormalizedLoopAge
      Scale Curve=&Multiply_Float(A=0.1, B=&Lerp_Float(A=1, B=3, Alpha=$System.IncomingAudio))
     )
    )
    Particles.GlowPosition:NiagaraPosition=&AddVectorToPosition(Position=$Output.SocketLocation.SampledPosition, Vector=<0 1 1>)
   ) {isSetParameters=true}
   /SetVariables(
    Particles.FlameSideCameraOffset:NiagaraFloat=&Lerp_Float(A=2, B=6, Alpha=$System.IncomingAudio)
    Particles.MeshSizeFireSide:Vector3f=&MakeVector(
     X=1
     Y=&Multiply_Float(
      A=1
      B=&SelectFloatFromArray(Direct Array Index=$Particles.PartitionParticles.Partition, Float Selection Array=DataInterface() {propertyValues {FloatData [1,0.8]}})
     )
     Z=&Multiply_Float(
      A=&Multiply_Float(
       A=1
       B=&SelectFloatFromArray(Direct Array Index=$Particles.PartitionParticles.Partition, Float Selection Array=DataInterface() {propertyValues {FloatData [1.2,0.8]}})
      )
      B=&Lerp_Float(A=0.8, B=1.45, Alpha=$System.IncomingAudio)
     )
    )
    Particles.SideSocketOffset:NiagaraPosition=&AddVectorToPosition(Position=$Output.SocketLocation.SampledPosition, Vector=<0 -0.5 -1.5>)
   ) {isSetParameters=true}
   &FN_Light_Attributes(
    Color=<rgba 1 0.2109 0 1>
    Create Color Attribute=true
    Create Exponent Attribute=true
    Create Radius Attribute=true
    Enable Particle Light Enable=true
    Exponent:NiagaraFloat=&Lerp_Float(A=4, B=2, Alpha=$System.IncomingAudio)
    Radius=30
    Scale Color=&SelectFloatFromArray(Direct Array Index=$Particles.PartitionParticles.Partition, Float Selection Array=DataInterface() {propertyValues {FloatData [15,15]}})
   )
   /SetVariables(
    Particles.LightCOlor:LinearColor=&Multiply_LinearColorByFloat(
     Linear Color=$Particles.FN_Light_Attributes.Color
     Float=&FloatFromCurve(
      FloatCurve=DataInterface() {propertyValues {Curve {keys [
          <at 0 1 RCIM_Cubic>
          {arriveTangent=0.0881, interpMode=RCIM_Cubic, leaveTangent=0.0881, time=0.2556, value=0.6881}
          {arriveTangent=-0.3011, interpMode=RCIM_Cubic, leaveTangent=-0.3011, time=0.4826, value=1.0425}
          {arriveTangent=0.1649, interpMode=RCIM_Cubic, leaveTangent=0.1649, time=0.6861, value=0.5584}
          {arriveTangent=0.8968, interpMode=RCIM_Cubic, leaveTangent=0.8968, time=0.7888, value=1.093}
          {arriveTangent=-0.4403, interpMode=RCIM_Cubic, leaveTangent=-0.4403, time=0.9253, value=0.773}
          <at 1 1 RCIM_Cubic>
         ]}}}
      CurveIndex=$System.NormalizedLoopAge
     )
    )
    Particles.LightPos:NiagaraPosition=&AddVectorToPosition(
     Position=$Output.SocketLocation.SampledPosition
     Vector=&SelectVectorFromArray(
      Direct Array Index=$Particles.PartitionParticles.Partition
      Vector Selection Array=DataInterface() {propertyValues {FloatData [<1 0 3>, <0 0 4>]}}
     )
    )
   ) {isSetParameters=true}
  ]
  rendererClasses [
   &NiagaraMeshRendererProperties
   &NiagaraSpriteRendererProperties
   &NiagaraLightRendererProperties
  ]
  renderers [
   {rendererClass=&NiagaraMeshRendererProperties, rendererIndex=0}
   {rendererClass=&NiagaraMeshRendererProperties, rendererIndex=1}
   {rendererClass=&NiagaraMeshRendererProperties, rendererIndex=2}
   {rendererClass=&NiagaraMeshRendererProperties, rendererIndex=3}
   {rendererClass=&NiagaraSpriteRendererProperties, rendererIndex=4}
   {rendererClass=&NiagaraMeshRendererProperties, rendererIndex=5}
   {rendererClass=&NiagaraMeshRendererProperties, rendererIndex=6}
   {rendererClass=&NiagaraLightRendererProperties, rendererIndex=7}
  ]
 }
 RackDecal {
  emitterUpdateScript [
   &EmitterState(Life Cycle Mode=::NewEnumerator1, Loop Behavior=::NewEnumerator1, Loop Duration Mode=::NewEnumerator1, UseLoopDelay=false)
   &SpawnBurst_Instantaneous(
    Spawn Count=&Multiply_Int(
     A=2
     B=&MakeIntFromFloat(Float=&ClampFloat(Float=&OneMinusFloat(Float=$User.HideFX)))
    )
    Use Spawn Probability=false
    UseLoopCountLimit=false
   )
  ]
  particleSpawnScript [
   &InitializeParticle(
    Color=<rgba 2 0.3833 0.1139 0.2>
    Sprite Rotation Angle=-25
    Sprite Rotation Mode=::NewEnumerator0
    Sprite Size Mode=::NewEnumerator3
    Uniform Sprite Size=20
    UsePositionOffset=false
    Write Material Random=true
   )
   &SpriteFacingAndAlignment(Sprite Alignment=<0 0 1>, Sprite Facing=<0 -1 0>, WriteAlignment=true, WriteFacing=true)
   &CameraOffset(Camera Offset Amount=10)
   &PartitionParticles()
  ]
  particleUpdateScript [
   &ParticleState(Kill Particles When Lifetime Has Elapsed=false)
   &SocketLocation(
    SamplePosOffset=true
    Socket Index=&SelectIntFromArray(Direct Array Index=$Particles.PartitionParticles.Partition, Int Selection Array=DataInterface() {propertyValues {IntData [6,7]}})
    Socket Reader=$System.Socket Reader
    Socket Sampling Mode=::NewEnumerator3
    VelocitySpeedLimit=false
   )
   &DynamicMaterialParameters(
    Index 0 Param 2=&Lerp_Float(A=0.75, B=0.3, Alpha=$System.IncomingAudio)
    Index 0 Param 3=&RandomRangeFloat(Minimum=0, Maximum=1)
    Module.Index0Float0Write=true
    Module.Index0Float1Write=true
    Module.Index0Float2Write=true
    Module.Index0Float3Write=true
   )
   &ScaleColor(
    Scale Alpha=0.35
    Scale RGB=&Multiply_VectorByFloat(Vector=<1 1 1>, Float=&Lerp_Float(A=1, B=2, Alpha=$System.IncomingAudio))
    ScaleA=true
    ScaleRGB=true
   )
  ]
  rendererClasses [&NiagaraSpriteRendererProperties]
  renderers [{rendererClass=&NiagaraSpriteRendererProperties}]
 }
 Sparks {
  emitterUpdateScript [
   &EmitterState()
   &SpawnRate(
    Spawn Probability=&ClampFloat(Float=&OneMinusFloat(Float=$User.HideFX))
    SpawnRate=&Lerp_Float(A=20, B=100, Alpha=$System.IncomingAudio)
    Use Spawn Probability=true
   )
  ]
  particleSpawnScript [
   &InitializeParticle(
    Lifetime Max=0.6
    Lifetime Min=0.3
    Lifetime Mode=::NewEnumerator1
    Sprite Size Max=<2.5 9>
    Sprite Size Min=<2 4>
    Sprite Size Mode=::NewEnumerator2
    UsePositionOffset=false
    Write Material Random=true
   )
   &PartitionParticles()
   &SocketLocation(
    SamplePosOffset=true
    Socket Index=&SelectIntFromArray(Direct Array Index=$Particles.PartitionParticles.Partition, Int Selection Array=DataInterface() {propertyValues {IntData [0,1]}})
    Socket Reader=$System.Socket Reader
    Socket Sampling Mode=::NewEnumerator3
    VelocitySpeedLimit=false
   )
   &ShapeLocation(
    Cylinder Height=18
    Cylinder Radius=2
    MaskSpawnGroup=false
    Override Seed=false
    Rotation Angle Type:ENiagara_AngleInput=::NewEnumerator0
    Rotation Mode=::NewEnumerator2
    Shape Primitive=::NewEnumerator1
    Surface Only=false
    UseOwnerScale=true
    Yaw / Pitch / Roll=<0 60 0>
   )
   &AddVelocity(
    Rotation Coordinate Space=::Simulation
    Velocity=&MakeVector(
     X=0
     Y=0
     Z=&RandomRangeFloat(Minimum=25, Maximum=&Lerp_Float(A=80, B=150, Alpha=$System.IncomingAudio))
    )
   )
  ]
  particleUpdateScript [
   &ParticleState()
   &ScaleColor(
    Scale Alpha=&FloatFromCurve(FloatCurve=DataInterface() {propertyValues {Curve {
        defaultValue=1
        keys [{interpMode=RCIM_Cubic, tangentMode=RCTM_User, time=0, value=1}, {interpMode=RCIM_Cubic, tangentMode=RCTM_User, time=1, value=0}]
       }}})
    Scale RGB=&MakeVectorFromLinearColorRGB(Color=&ColorFromCurve(ColorCurve=DataInterface() {propertyValues {
        AlphaCurve {keys [<at 0 1>]}
        BlueCurve {keys [<at 0 0.0729>,<at 0.7152 0.0319>]}
        GreenCurve {keys [<at 0 0.6395>,<at 0.7152 0.2178>]}
        RedCurve {keys [<at 0 2>,<at 0.7152 0.7604>]}
       }}))
    ScaleA=true
    ScaleRGB=true
   )
   &ScaleSpriteSize(Scale Factor=&Vector2DFromFloat(Value=&FloatFromCurve(
      FloatCurve=DataInterface() {propertyValues {Curve {keys [
          {arriveTangent=10.1867, interpMode=RCIM_Cubic, leaveTangent=10.1867, tangentMode=RCTM_User, time=0, value=0}
          {interpMode=RCIM_Cubic, leaveTangent=-0.0263, tangentMode=RCTM_Break, time=0.2, value=1.2}
          {arriveTangent=-2.7631, interpMode=RCIM_Cubic, leaveTangent=-2.7631, tangentMode=RCTM_User, time=1, value=0}
         ]}}}
      Scale Curve=&Multiply_Float(A=0.25, B=&FN_FrontEndFloatModifier(Gameplay=1, Front End=0.8))
     ))) {module=ScaleSpriteSize001}
   &CurlNoiseForce(Noise Frequency=100, Noise Strength=120, Pan Noise=false, Random Seed=10)
   &Drag(Drag=&RandomRangeFloat(Minimum=4, Maximum=6), Use Linear Drag=true, Use Rotational Drag=false)
   &GravityForce(Gravity=&RandomRangeVector(Minimum=<0 0 100>, Maximum=<0 0 200>))
   &SolveForcesAndVelocity(Clamp Velocity=false, Limit Acceleration=false)
   &ScaleSpriteSizeBySpeed(Max Scale Factor=<0.8 2>, Velocity Threshold=100)
   /SetVariables(Particles.Glow03_Color=&MultiplyLinearColor(
     Linear Color=$Particles.Color
     Scale Factor=&MakeVector4FromVectorAndFloat("Vector (RGB)"=&Multiply_VectorByFloat(Vector=<1 1 1>, Float=0.8), "Float (Alpha)"=0.5)
    )) {isSetParameters=true}
   /SetVariables(Particles.Glow03_Size=&Multiply_Vector2D(A=$Particles.SpriteSize, B=<4 2>)) {isSetParameters=true}
  ]
  rendererClasses [&NiagaraSpriteRendererProperties]
  renderers [
   {rendererClass=&NiagaraSpriteRendererProperties}
   {rendererClass=&NiagaraSpriteRendererProperties, rendererIndex=1}
  ]
 }
 WheelFlames {
  emitterUpdateScript [
   &EmitterState(Life Cycle Mode=::NewEnumerator1, Loop Behavior=::NewEnumerator1, Loop Duration Mode=::NewEnumerator1, UseLoopDelay=false)
   &SpawnBurst_Instantaneous(
    Spawn Count=&Multiply_Int(
     A=4
     B=&MakeIntFromFloat(Float=&ClampFloat(Float=&OneMinusFloat(Float=$User.HideFX)))
    )
    Use Spawn Probability=false
    UseLoopCountLimit=false
   )
  ]
  particleSpawnScript [
   &InitializeParticle(
    Color=<rgba 2 0.3437 0.0802 1>
    Lifetime=5
    Sprite Rotation Angle=-90
    Sprite Rotation Mode=::NewEnumerator0
    Sprite Size Mode=::NewEnumerator1
    Uniform Sprite Size Max=12
    Uniform Sprite Size Min=10
    UsePositionOffset=false
    Write Material Random=true
   )
   &PartitionParticles(Number Of Partitions=4)
   &SocketLocation(
    SamplePosOffset=true
    Socket Index=&SelectIntFromArray(Direct Array Index=$Particles.PartitionParticles.Partition, Int Selection Array=DataInterface() {propertyValues {IntData [2,3,4,5]}})
    Socket Reader=$System.Socket Reader
    Socket Sampling Mode=::NewEnumerator3
    VelocitySpeedLimit=false
   )
   &CameraOffset(Camera Offset Amount=3)
   /SetVariables(Particles.CameraOffset2=2) {isSetParameters=true}
   /SetVariables(
    Particles.FlameAlignment:Vector3f=<0 0 1>
    Particles.FlamePosition:NiagaraPosition=&AddVectorToPosition(
     Position=$Output.SocketLocation.SampledPosition
     Vector=&SelectVectorFromArray(
      Direct Array Index=$Particles.PartitionParticles.Partition
      Vector Selection Array=DataInterface() {propertyValues {FloatData [<0 1 1.5>, <0.65 1 1>, <0.6 1 1>, <-0.3 1 2>]}}
     )
    )
    Particles.FlameSize:Vector2f=&SelectVector2DFromArray(
     Direct Array Index=$Particles.PartitionParticles.Partition
     Vector 2D Selection Array=DataInterface() {propertyValues {FloatData [<9 5>, <14 5>, <8 5>, <6 5>]}}
    )
   ) {isSetParameters=true}
  ]
  particleUpdateScript [
   &ParticleState(Kill Particles When Lifetime Has Elapsed=false)
   &DynamicMaterialParameters(
    Index 0 Param 1=0.8
    Index 0 Param 2=&Lerp_Float(A=0.75, B=0.3, Alpha=$System.IncomingAudio)
    Index 0 Param 3=&RandomRangeFloat(Minimum=0, Maximum=1)
    Module.Index0Float0Write=true
    Module.Index0Float1Write=true
    Module.Index0Float2Write=true
    Module.Index0Float3Write=true
   )
   /SetVariables(Particles.DMPFlame:Vector4f=<0.8 0.25 0 0>) {isSetParameters=true}
   &ScaleColor(
    Scale RGB=&Multiply_VectorByFloat(Vector=<1 1 1>, Float=&Lerp_Float(A=1, B=2, Alpha=$System.IncomingAudio))
    ScaleA=true
    ScaleRGB=true
   )
   /SetVariables(
    Particles.FlameSize:Vector2f=&Multiply_Vector2D(
     A=$Particles.Initial.FlameSize
     B=&MakeVector2D(
      X=&Lerp_Float(A=1, B=1.5, Alpha=$System.IncomingAudio_R)
      Y=&Lerp_Float(A=1, B=1.2, Alpha=$System.IncomingAudio)
     )
    )
    Particles.FlamesColor:LinearColor=&MakeLinearColorFromVectorAndFloat(
     "Vector (RGB)"=&Multiply_VectorByFloat(
      Vector=&MakeVectorFromLinearColorRGB(Color=$Particles.Initial.FlamesColor)
      Float=&Lerp_Float(A=1, B=1.2, Alpha=$System.IncomingAudio)
     )
     "Float (Alpha)"=1
    )
    Particles.SmokeSize:Vector2f=&SelectVector2DFromArray(
     Direct Array Index=$Particles.PartitionParticles.Partition
     Vector 2D Selection Array=DataInterface() {propertyValues {FloatData [<15 8>, <15 8>, <15 8>, <15 8>]}}
    )
   ) {isSetParameters=true}
  ]
  rendererClasses [&NiagaraSpriteRendererProperties]
  renderers [
   {rendererClass=&NiagaraSpriteRendererProperties}
   {rendererClass=&NiagaraSpriteRendererProperties, rendererIndex=1}
   {rendererClass=&NiagaraSpriteRendererProperties, rendererIndex=2}
   {rendererClass=&NiagaraSpriteRendererProperties, rendererIndex=3}
  ]
 }
}
summary {
 systemName=NS_Backpack_NoiseClueGust_Idle
 emitters [
  {emitterName=Sparks, rendererClasses [&NiagaraSpriteRendererProperties]}
  {emitterName=RackDecal, rendererClasses [&NiagaraSpriteRendererProperties]}
  {
   emitterName=FlameMeshFront
   rendererClasses [
    &NiagaraMeshRendererProperties
    &NiagaraSpriteRendererProperties
    &NiagaraLightRendererProperties
   ]
  }
  {emitterName=WheelFlames, rendererClasses [&NiagaraSpriteRendererProperties]}
  {emitterName=FlameEdge, rendererClasses [&NiagaraMeshRendererProperties, &NiagaraSpriteRendererProperties]}
 ]
 userVariables [
  User.bIsFrontEnd {type=NiagaraBool}
  User.bIsFrontEndPreview {type=NiagaraBool}
  User.blsSwinging {type=NiagaraBool}
  User.DebugReactivity {type=NiagaraBool}
  User.EnableFrontendFakeAudioLoop {type=NiagaraFloat, defaultValue=0, description=""}
  User.HideFX {type=NiagaraFloat, defaultValue=0, description=""}
  User.Reactivity {type=NiagaraFloat, defaultValue=1, description=""}
  User.RibbonIDOffset {type=NiagaraInt32}
 ]
}
